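# Masked material for the Watermelon cosmetic wheel, built on MF_WheelParent_Simple.
# Vertex red separates rim from tire for normals, specular and camera-facing toon bands in emissive, stepped with Ceil.
# In vertex-blue areas, rim texture alpha cuts opacity and overlays color, its UVs jittered by zero-tiling panned gradients amplified and clamped into snapping offsets.
# The result passes through vehicle damage and de-rez transition functions.
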Material /VehicleCosmetics/Wheels/Watermelon/Materials/M_Wheel_Watermelon
graph {
 properties {
  bAutomaticallySetUsageInEditor=false
  bUseMaterialAttributes=true
  bUsedWithMeshDeformer=true
  bUsedWithSkeletalMesh=true
  blendMode=BLEND_Masked
  pixelDepthOffsetMode=PDOM_Legacy
  usedShadingModels=Default Lit
 }
 nodes [
  n1 = CameraVectorWS()
  n2 = TextureSampleParameter2D() {parameterName=Tire_Normal, samplerType=SAMPLERTYPE_Normal, texture=/VehicleCosmetics/Textures/T_BlankNormal_B}
  n3 = VertexColor()
  n4 = StaticSwitchParameter(
   True=LinearInterpolate(
    A=TextureSampleParameter2D() {parameterName=Rim_Normal, samplerType=SAMPLERTYPE_Normal, texture=&T_BlankNormal_B}.RGB
    B=@n2.RGB
    Alpha=@n3.R
   )
   False=Constant3Vector() {constant=<rgba 0 0 1 0>}
  ) {defaultValue=true, group=6_ToonProperty, parameterName=UseNormal}
  DotProduct(A=Transform(None=@n1) {transformSourceType=TRANSFORMSOURCE_Local, transformType=TRANSFORM_Tangent}, B=@n4)
  n5 = GetMaterialAttributes(None=MaterialFunctionCall() {materialFunction=&MF_WheelParent_Simple}.MaterialAttributes) {attributeGetTypes [#1, #2]}
  n6 = TextureCoordinate() {uTiling=0, vTiling=0}
  n7 = Constant() {r=1200}
  n8 = Multiply(
   A=TextureSample(UVs=Panner(Coordinate=@n6) {speedX=0.2, speedY=0.2}) {texture=/VehicleCosmetics/Wheels/_SharedTextures/FX/MowGradient01}.R
   B=@n7
  )
  n9 = Multiply(
   A=TextureSample(UVs=Panner(Coordinate=@n6) {speedX=-0.1, speedY=-0.1}) {texture=&MowGradient01}.R
   B=@n7
  )
  n10 = ConstantBiasScale(None=AppendVector(A=Clamp(None=@n8), B=Clamp(None=@n9))) {bias=-0.5, scale=0.0175}
  n11 = TextureSample() {texture=/VehicleCosmetics/Wheels/Watermelon/Textures/T_Rim_Watermelon_BC}
  n12 = TextureSample(UVs=Add(A=TextureCoordinate(), B=Multiply(A=@n10, B=@n11.A))) {texture=&T_Rim_Watermelon_BC}
  n13 = Multiply(A=TextureSample() {texture=&T_Rim_Watermelon_BC}.A, B=VertexColor().B)
  n14 = VertexColor()
  n15 = OneMinus(None=TextureSample() {texture=&T_Rim_Watermelon_BC}.A)
  n16 = TextureSampleParameter2D() {group=6_ToonProperty, parameterName=Rim_DiffuseTextureBlend, texture=/VehicleCosmetics/Textures/T_White}
  n17 = ScalarParameter() {defaultValue=1, group=6_ToonProperty, parameterName=BlendPower}
  n18 = DotProduct(A=@n1, B=Transform(None=@n4))
  n19 = Multiply(A=@n18, B=ScalarParameter() {defaultValue=2, group=6_ToonProperty, parameterName=Rim_Stepcount})
  n20 = ScalarParameter() {defaultValue=0.2, group=6_ToonProperty, parameterName=Rim_StepAmount}
  n21 = ConstantBiasScale(None=Ceil(None=Subtract(A=@n19, B=@n20))) {bias=0.05, scale=5}
  n22 = ScalarParameter() {defaultValue=2, group=6_ToonProperty, parameterName=Tire_Stepcount}
  n23 = Ceil(None=Subtract(A=Multiply(A=@n18, B=@n22), B=ScalarParameter() {defaultValue=0.2, group=6_ToonProperty, parameterName=Tire_StepAmount}))
  n24 = LinearInterpolate(A=Multiply(A=Multiply(A=@n16.RGB, B=@n17), B=Clamp(None=@n21)), B=Clamp(None=ConstantBiasScale(None=@n23) {bias=-0.4, scale=1}), Alpha=@n3.R)
  n25 = Constant() {r=1}
  n26 = ScalarParameter() {defaultValue=40, group=6_ToonProperty, parameterName=SpecPower}
  n27 = SetMaterialAttributes!(
   MaterialAttributes=@n5.MaterialAttributes
   Base Color=LinearInterpolate(A=@n5.Base Color, B=@n12.RGB, Alpha=@n13)
   Opacity Mask=Subtract(A=@n5.Opacity Mask, B=Multiply(A=@n14.B, B=@n15))
   Emissive Color=LinearInterpolate(A=Multiply(A=@n5.Base Color, B=@n24), B=@n12.RGB, Alpha=@n13)
   Specular=Power(
    Base=Desaturation(None=@n2.RGB, Fraction=@n25) {luminanceFactors=<rgba 0.3 0.59 0.11 0>}
    Exp=LinearInterpolate(A=@n26, Alpha=@n3.R) {constB=6}
   )
  ) {attributeSetTypes [
    #1
    #2
    #3
    #4
   ]}
  MaterialFunctionCall(In=MaterialFunctionCall(Material=@n27) {materialFunction=&MF_BaseVehicle_Damage}.Result) {materialFunction=&MF_VehicleDeRez_Transition}
 ]
}
